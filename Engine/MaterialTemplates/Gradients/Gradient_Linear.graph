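# Template material that builds a linear gradient along the U texture coordinate, using the Default Lit shading model.
# A SphereMask compares the red channel of the UVs to a constant midpoint, giving a soft band that peaks at the center and fades toward both sides.
# The mask feeds the diffuse color input of a ConvertFromDiffSpec function, with no textures or parameters involved.
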Material /Engine/MaterialTemplates/Gradients/Gradient_Linear
graph {
 properties {pixelDepthOffsetMode=PDOM_Legacy, usedShadingModels=Default Lit}
 nodes [
  n1 = TextureCoordinate()
  MaterialFunctionCall(DiffuseColor=SphereMask(A=ComponentMask(None=@n1) {r=true}, B=Constant() {r=0.5}) {attenuationRadius=0.4, desc=AttenuationRadius .4, hardnessPercent=0}) {materialFunction=&ConvertFromDiffSpec}
 ]
}
