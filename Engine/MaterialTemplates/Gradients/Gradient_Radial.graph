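# Radial gradient material template that builds a soft circular falloff across UV space, using the Default Lit shading model.
# A SphereMask measures texture coordinates against a constant center at the middle of UV space, with zero hardness for a smooth fade toward the edge of its radius.
# The mask feeds the DiffuseColor input of the ConvertFromDiffSpec material function.
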Material /Engine/MaterialTemplates/Gradients/Gradient_Radial
graph {
 properties {pixelDepthOffsetMode=PDOM_Legacy, usedShadingModels=Default Lit}
 nodes [MaterialFunctionCall(DiffuseColor=SphereMask(A=TextureCoordinate(), B=Constant2Vector() {g=0.5, r=0.5}) {attenuationRadius=0.4, desc=AttenuationRadius .4, hardnessPercent=0}) {materialFunction=&ConvertFromDiffSpec}]
}
