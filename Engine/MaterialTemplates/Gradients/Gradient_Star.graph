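# Material template that builds a procedural star-shaped gradient from texture coordinates, with no textures.
# Offsetting the UVs by half, applying Ceil and offsetting back gives each pixel its nearest tile corner as a SphereMask center.
# The soft corner falloff is inverted and raised to a high power, leaving a bright center with pinched arms that repeats every UV tile.
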Material /Engine/MaterialTemplates/Gradients/Gradient_Star
graph {
 properties {pixelDepthOffsetMode=PDOM_Legacy, usedShadingModels=Default Lit}
 nodes [
  n1 = Add(A=TextureCoordinate(), B=Constant2Vector() {g=0.5, r=0.5})
  n2 = Add(A=Ceil(None=@n1), B=Constant2Vector() {g=-0.5, r=-0.5})
  Power(Base=OneMinus(None=SphereMask(A=@n1, B=@n2) {attenuationRadius=0.8, desc=AttenuationRadius .5, hardnessPercent=0}), Exp=Constant() {r=8})
 ]
}
